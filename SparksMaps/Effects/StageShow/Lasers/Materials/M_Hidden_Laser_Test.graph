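# Unlit masked material for Niagara mesh particles in the stage show laser effects, named as a hidden laser test.
# The graph holds only a single default Constant node, with no connection to any material output shown.
# It has no parameters, textures or animation logic.
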Material /SparksMaps/Effects/StageShow/Lasers/Materials/M_Hidden_Laser_Test
graph {
 properties {
  bUsedWithNiagaraMeshParticles=true
  blendMode=BLEND_Masked
  pixelDepthOffsetMode=PDOM_Legacy
  shadingModel=MSM_Unlit
  usedShadingModels=Unlit
 }
 nodes [Constant()]
}
