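# Additive, unlit, two-sided material for Sparks stage show lasers on Niagara mesh particles.
# An IsBeam switch selects edge-faded, spline-thickened lines or a softer gradient, tinted by particle color or MojoStageColors.
# Scrolling stabilized screen-space smoke noise adds to a profile set by dynamic parameters, with optional blending toward animated oscillating noise.
# Depth fades soften the emissive result, with named reroutes for a color boost and audio BeatLifetime modulation.
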Material /SparksMaps/Effects/StageShow/Lasers/Materials/M_Sparks_Laser_Show_B
graph {
 properties {
  bUsedWithNiagaraMeshParticles=true
  blendMode=BLEND_Additive
  pixelDepthOffsetMode=PDOM_Legacy
  shadingModel=MSM_Unlit
  twoSided=true
  usedShadingModels=Unlit
 }
 nodes [
  Constant() {r=0.25}
  Constant() {r=1024}
  Constant()
  n1 = ComponentMask(None=TextureCoordinate()) {r=true}
  n2 = Multiply(A=@n1, B=OneMinus(None=@n1))
  n3 = OneMinus(None=@n2)
  n4 = Multiply(A=Power(Base=@n2) {constExponent=1.75}) {constB=28}
  n5 = DynamicParameter() {parameterIndex=1, defaultValue=<rgba 0 0.09 -0.01 1>, paramNames [LaserSMokeDetailSelect, Scroll Speed X, Scroll Speed Y, Color Mult]}
  n6 = LinearInterpolate(A=Multiply(A=Power(Base=@n3) {constExponent=8}) {constB=4}, B=@n4, Alpha=Saturate(None=@n5.LaserSMokeDetailSelect))
  LinearInterpolate(Alpha=Saturate(None=@n6)) {constB=2}
  n7 = Constant3Vector()
  n8 = MaterialFunctionCall(Float3=Multiply(A=TransformPosition(None=@n7) {transformType=TRANSFORMPOSSOURCE_World}) {constB=0.0001}) {materialFunction=&BreakOutFloat3Components}
  MaterialFunctionCall(
   Hue Shift Percentage=TextureSample(UVs=MaterialFunctionCall(R=@n8.R, G=@n8.R) {materialFunction=&AppendMany}.RG) {texture=/Game/Global/Textures/Default/Noise/3xPerlin_M}.RGB
   Texture=Constant3Vector() {constant=<rgba 1 0 0 0>}
  ) {materialFunction=&HueShift}
  n9 = Saturate(None=NamedRerouteUsage())
  n10 = TextureCoordinate()
  n11 = ComponentMask(None=@n10) {r=true}
  n12 = Multiply(A=@n11, B=OneMinus(None=@n11))
  n13 = ScalarParameter() {defaultValue=1.25, parameterName=Edges Fade Power}
  n14 = ComponentMask(None=@n10) {g=true}
  n15 = ScalarParameter() {defaultValue=3, parameterName=Sharp Line End Fade}
  n16 = DynamicParameter() {defaultValue=<rgba 0 8 35 1>, paramNames [UseCustomColor, LaserThicknessBase, LaserThicknessTip, Color Lut Custom U Position]}
  n17 = ParticleColor()
  n18 = LinearInterpolate(
   A=@n17.RGB
   B=MaterialFunctionCall!(CustomUPosition=@n16.Color Lut Custom U Position) {materialFunction=/SparksStageShow/Systems/StageManager/FXManager/MaterialFunctions/MF_MojoStageColors}.Primary
   Alpha=@n16.UseCustomColor
  )
  n19 = Multiply(A=Multiply(A=Power(Base=@n12, Exp=@n13), B=Power(Base=@n14, Exp=@n15)), B=@n18)
  n20 = MaterialFunctionCall(
   LDR Input=Saturate(None=@n19)
   HDR Tint=@n19
   HDR Intensity Multiplier=ScalarParameter() {defaultValue=10, parameterName=HDR Thin_Lines_Multiplier}
  ) {materialFunction=&DeriveHDRfromLDR}
  n21 = ComponentMask(None=TextureCoordinate()) {g=true}
  n22 = Power(Base=@n21) {constExponent=4}
  n23 = Multiply(A=Power(Base=@n21)) {constB=0.05}
  n24 = Multiply(A=@n18, B=Multiply(A=Add(A=@n22, B=@n23)) {constB=0.25})
  n25 = MaterialFunctionCall(LDR Input=Saturate(None=@n24), HDR Tint=@n24, HDR Intensity Multiplier=ScalarParameter() {defaultValue=1.4, parameterName=HDR Multiplier}) {materialFunction=&DeriveHDRfromLDR}
  n26 = MaterialFunctionCall() {materialFunction=&CameraDepthFade}
  n27 = Multiply(A=Multiply(A=StaticSwitchParameter(True=@n20.RemappedRange, False=@n25.RemappedRange) {parameterName=IsBeam}, B=Multiply(A=@n17.A, B=@n26.Result))) {constB=2.5}
  n28 = Constant() {r=1}
  n29 = ScalarParameter() {defaultValue=4, parameterName=Noise UV Scale}
  n30 = Multiply(A=Time(), B=LinearInterpolate(Alpha=ParticleRandom()) {constA=0.5, constB=1.5})
  n31 = MaterialFunctionCall!(In) {materialFunction=/Game/Weapons/FORT_Melee/Pickaxe_Parfait_Female/FX/Materials/Functions/MF_TrifleParfait_BrightnessOscillation}
  n32 = Time()
  n33 = AppendVector(A=@n5.Scroll Speed X, B=@n5.Scroll Speed X)
  n34 = MaterialFunctionCall!(Position, UVRadialScale=Constant() {r=0.235}) {materialFunction=/Game/Weapons/WeaponSkins/MaterialFunction/ObjectStabilizedScreenSpaceUVs}
  n35 = TextureSampleParameter2D(UVs=Add(A=@n34.StabilizedUVs, B=Multiply(A=@n32, B=@n33))) {
   parameterName=WS Smoke Noise
   samplerType=SAMPLERTYPE_LinearColor
   texture=/Game/VisualThreatManager/StormVisuals/Textures/Clouds_VectorDisplacement
  }
  n36 = Power(Base=@n35.R) {constExponent=4}
  n37 = LinearInterpolate(
   A=Add(A=@n36, B=@n6)
   B=Multiply(
    A=MaterialFunctionCall!(Noise Intensity=@n28, Noise UV Scale=@n29, Time=@n30, Noise Axis Rotation) {materialFunction=/Game/Weapons/FORT_Melee/Pickaxe_Parfait_Female/FX/Materials/Functions/MF_TrifleParfait_Noise}.Noise Width Multiplier
    B=Power(Base=@n31.Brightness) {constExponent=4}
   )
   Alpha=ScalarParameter() {parameterName=Noise Lerp}
  ) {constAlpha=0}
  n38 = Multiply(
   A=Multiply(A=LinearInterpolate(Alpha=@n9) {constA=1, constB=0.25}, B=Multiply(A=@n37, B=@n27))
   B=DepthFade(FadeDistance=ScalarParameter() {defaultValue=150, parameterName=Depth Fade})
  )
  n39 = MaterialFunctionCall() {materialFunction=&CameraDepthFade}
  MaterialFunctionCall(Emissive=Multiply(A=Multiply(A=@n38, B=@n39.Result), B=NamedRerouteUsage())) {materialFunction=&MF_FX_EmissiveAdjustments}
  Multiply(A=@n36)
  Multiply(A=@n4, B=@n2)
  NamedRerouteDeclaration(None=@n5.Color Mult) {name=boost, nodeColor=<rgba 0.5469 1 0 1>}
  NamedRerouteDeclaration(None=CollectionParameter() {collection=/FMAudioReactiveData/AudioAnalysis/MPC_Sparks_Audio_Parameters, parameterName=BeatLifetime}) {name=AudioValue, nodeColor=<rgba 1 0 0.8203 1>}
  Panner(
   Coordinate=@n34.StabilizedUVs
   Speed=AppendVector(
    A=ScalarParameter() {defaultValue=0.09, parameterName=WS Smoke Speed X}
    B=ScalarParameter() {defaultValue=-0.01, parameterName=WS Smoke Speed Y}
   )
  ) {bFractionalPart=true, speedX=0.09, speedY=-0.1}
  Saturate(None=@n4)
  ScalarParameter() {defaultValue=16, parameterName=LaserThicknessTip}
  ScalarParameter() {defaultValue=16, parameterName=LaserThicknessBase}
  StaticSwitchParameter(
   True=MaterialFunctionCall(WidthBase=@n16.LaserThicknessBase, WidthTip=@n16.LaserThicknessBase) {materialFunction=&SplineThicken}.WorldPositionOffset
   False=Constant()
  ) {parameterName=IsBeam}
 ]
}
